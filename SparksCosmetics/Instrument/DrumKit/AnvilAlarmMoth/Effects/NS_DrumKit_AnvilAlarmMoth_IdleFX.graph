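# Idle Niagara effect for the AnvilAlarmMoth drum kit cosmetic, sending mesh particles spiraling up from skeletal mesh bones or sockets.
# TravelingLoop emitters burst long-lived particles and loop them along procedural helices, using execution index plus time as path position.
# A fake audio-reactivity loop supplies low, medium and high bands that pulse scale, tint and material parameters per particle partition.
# The earlier Constant and SpawnRate spiral emitters are disabled.
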NiagaraSystem /SparksCosmetics/Instrument/DrumKit/AnvilAlarmMoth/Effects/NS_DrumKit_AnvilAlarmMoth_IdleFX
dependencies {
 usedDataInterfaces [
  &NiagaraDataInterfaceCurve
  &NiagaraDataInterfaceColorCurve
  &NiagaraDataInterfaceSkeletalMesh
  &NiagaraDataInterfaceArrayFloat
 ]
 usedDynamicInputs [
  &MakeFloatFromVector
  &Lerp_Vector
  &MakeCustomFloatFromBool
  /Niagara/DynamicInputs/Lerp/SmoothLerpOverTime_Float
  &RemapRange
  &SetBoolByFloatComparison
  &FNTimeOfDayScalar
 ]
 usedModules [
  &self:SystemSpawnScript.SystemScriptSource.SystemScriptGraph.NiagaraNodeAssignment_12.SetVariables_#1
  &self:SystemSpawnScript.SystemScriptSource.SystemScriptGraph.NiagaraNodeAssignment_0.SetVariables_#2
  &SystemState
  &Timeline
  /Game/Effects/Niagara/Modules/Audio/FN_FakeAudioReactivityLoop
  &self:SystemSpawnScript.SystemScriptSource.SystemScriptGraph.NiagaraNodeAssignment_11.SetVariables_#3
  &self:SystemSpawnScript.SystemScriptSource.SystemScriptGraph.NiagaraNodeAssignment_13.SetVariables_#4
  &self:SystemSpawnScript.SystemScriptSource.SystemScriptGraph.NiagaraNodeAssignment_14.SetVariables_#5
  &self:SystemSpawnScript.SystemScriptSource.SystemScriptGraph.NiagaraNodeAssignment_6.SetVariables_#6
  &self:SystemSpawnScript.SystemScriptSource.SystemScriptGraph.NiagaraNodeAssignment_8.SetVariables_#7
  &self:SystemSpawnScript.SystemScriptSource.SystemScriptGraph.NiagaraNodeAssignment_1.SetVariables_#8
 ]
}
emitters {
 Spiral_Particles_Constant {
  bEnabled=false
  emitterUpdateScript [
   &EmitterState()
   &SpawnBurst_Instantaneous(Spawn Count=25, Use Spawn Probability=false, UseLoopCountLimit=true)
   /SetVariables(Emitter.Height=115, Emitter.SpiralFrequency=2, Emitter.SpiralSpeed=0.12) {isSetParameters=true, module=SetVariables_#9}
  ]
  particleSpawnScript [&InitializeParticle(
    Lifetime=5
    Mesh Scale Mode=::NewEnumerator1
    Mesh Uniform Scale Max=0.8
    Mesh Uniform Scale Min=0.6
    Sprite Size Mode=::NewEnumerator1
    Uniform Sprite Size Max=7
    UsePositionOffset=true
    Write Material Random=true
   )]
  particleUpdateScript [
   &ParticleState(Kill Particles When Lifetime Has Elapsed=false, Let Infinitely Lived Particles Die When Emitter Deactivates=true)
   /SetVariables(Particles.SpiralStrength:NiagaraFloat=&FloatFromCurve(
     FloatCurve=DataInterface() {propertyValues {Curve {
        defaultValue=1
        keys [
         {interpMode=RCIM_Cubic, leaveTangentWeight=0.2393, tangentMode=RCTM_User, time=0, value=1}
         {arriveTangent=-0.9582, interpMode=RCIM_Cubic, leaveTangent=-0.9582, tangentMode=RCTM_User, time=0.461, value=0.6565}
         <at 1 0.3093 RCIM_Cubic>
        ]
       }}}
     CurveIndex=&ReturnNormalizedExecIndex()
     Scale Curve=&Waveform(
      Global Input Position=&ReturnNormalizedExecIndex(Normalized Index Scale=0.5)
      "Waveform [1]"=::NewEnumerator0
      "[1] Phase Scale"=0.25
      "[1] Frequency"=1
      "[1] Amplitude Min/Max"=<0.5 1>
      "[1] Scale"=17
     )
    )) {isSetParameters=true, module=SetVariables_#10}
   &ScaleSpriteSize(
    Uniform Curve Index=&ReturnNormalizedExecIndex()
    Uniform Curve Scale=&Waveform(
     Global Input Position=&OneMinusFloat(Float=&ReturnNormalizedExecIndex())
     "Waveform [1]"=::NewEnumerator0
     "[1] Phase Scale"=0.5
     "[1] Frequency"=1
     "[1] Amplitude Min/Max"=<0.5 1>
    )
    Uniform Curve Sprite Scale=DataInterface() {propertyValues {Curve {
       defaultValue=0
       keys [
        {arriveTangent=4.5158, interpMode=RCIM_Cubic, leaveTangent=4.5158, tangentMode=RCTM_User, time=0, value=0.25}
        {arriveTangentWeight=0.1782, interpMode=RCIM_Cubic, leaveTangentWeight=0.1222, tangentMode=RCTM_User, time=0.2966, value=1}
        {arriveTangentWeight=0.0678, interpMode=RCIM_Cubic, leaveTangentWeight=0.0669, tangentMode=RCTM_User, time=0.5017, value=0.3444}
        {arriveTangentWeight=0.1206, interpMode=RCIM_Cubic, leaveTangentWeight=0.1799, tangentMode=RCTM_User, time=0.7007, value=1}
        {arriveTangent=-4.5159, tangentMode=RCTM_User, time=1, value=0.25}
       ]
      }}}
   )
   /SetVariables(Particles.PositionOffset:NiagaraPosition=&AddVectorToPosition(
     Position=&AddVectorToPosition(
      Position=<0 0 6>
      Vector=&Multiply_VectorByFloat(
       Vector=&RandomRangeVector(Minimum=<-1 -1 -1>, Maximum=<1 1 1>)
       Float=&FloatFromCurve(
        FloatCurve=DataInterface() {propertyValues {Curve {
           defaultValue=1
           keys [
            {interpMode=RCIM_Cubic, tangentMode=RCTM_User, time=0, value=0.1353}
            {arriveTangentWeight=0.3222, interpMode=RCIM_Cubic, leaveTangentWeight=0.0504, tangentMode=RCTM_User, time=0.25, value=1}
            {arriveTangentWeight=0.0226, interpMode=RCIM_Cubic, leaveTangentWeight=0.0234, tangentMode=RCTM_User, time=0.5, value=0.4235}
            {arriveTangentWeight=0.0523, interpMode=RCIM_Cubic, leaveTangentWeight=0.3204, tangentMode=RCTM_User, time=0.75, value=1}
            {tangentMode=RCTM_User, time=1, value=0.1353}
           ]
          }}}
        CurveIndex=&ReturnNormalizedExecIndex()
        Scale Curve=&Waveform(
         Global Input Position=&ReturnNormalizedExecIndex()
         "Waveform [1]"=::NewEnumerator2
         "[1] Phase Scale"=0.1
         "[1] Frequency"=1.1
         "[1] Amplitude Min/Max"=<-1 1>
         "[1] Scale"=6
        )
       )
      )
     )
     Vector=&MakeVector(
      X=&Waveform(
       Global Input Position=&ReturnNormalizedExecIndex()
       Global Phase=&Multiply_Float(A=$Engine.Time, B=$Emitter.SpiralSpeed)
       "Waveform [1]"=::NewEnumerator0
       "[1] Frequency"=$Emitter.SpiralFrequency
       "[1] Amplitude Min/Max"=<-1 1>
       "[1] Scale"=$Particles.SpiralStrength
      )
      Y=&Waveform(
       Global Input Position=&ReturnNormalizedExecIndex()
       Global Phase=&Multiply_Float(A=$Engine.Time, B=$Emitter.SpiralSpeed)
       "Waveform [1]"=::NewEnumerator1
       "[1] Frequency"=$Emitter.SpiralFrequency
       "[1] Amplitude Min/Max"=<-1 1>
       "[1] Scale"=$Particles.SpiralStrength
      )
      Z=&FloatFromCurve(
       FloatCurve=DataInterface() {propertyValues {Curve {defaultValue=0, keys [{tangentMode=RCTM_User, time=0, value=0}, {tangentMode=RCTM_User, time=1, value=1}]}}}
       CurveIndex=&ReturnNormalizedExecIndex()
       Scale Curve=$Emitter.Height
      )
     )
    )) {isSetParameters=true, module=SetVariables_#11}
   &ScaleMeshSize(Scale Factor=&Multiply_VectorByFloat(
     Vector=<1 1 1>
     Float=&Multiply_Float(
      A=$System.EngineOwnerScale
      B=&FloatFromCurve(
       FloatCurve=DataInterface() {propertyValues {Curve {
          defaultValue=0
          keys [
           {arriveTangent=4.5158, interpMode=RCIM_Cubic, leaveTangent=4.5158, tangentMode=RCTM_User, time=0, value=0.3794}
           {arriveTangentWeight=0.1782, interpMode=RCIM_Cubic, leaveTangentWeight=0.1222, tangentMode=RCTM_User, time=0.2966, value=1}
           {arriveTangentWeight=0.0678, interpMode=RCIM_Cubic, leaveTangentWeight=0.0669, tangentMode=RCTM_User, time=0.5017, value=0.6}
           {arriveTangentWeight=0.1206, interpMode=RCIM_Cubic, leaveTangentWeight=0.1799, tangentMode=RCTM_User, time=0.7007, value=1}
           {arriveTangent=-4.5159, tangentMode=RCTM_User, time=1, value=0.3794}
          ]
         }}}
       CurveIndex=&ReturnNormalizedExecIndex()
       Scale Curve=&Waveform(
        Global Input Position=&OneMinusFloat(Float=&ReturnNormalizedExecIndex(Normalized Index Scale=2))
        "Waveform [1]"=::NewEnumerator0
        "[1] Phase Scale"=0.5
        "[1] Frequency"=1
        "[1] Amplitude Min/Max"=<0.7 1>
       )
      )
     )
    ))
   &ScaleColor(
    Scale RGB=&MakeVectorFromLinearColorRGB(Color=&ColorFromCurve(
      ColorCurve=DataInterface() {propertyValues {
        AlphaCurve {keys [<at 0 1>]}
        BlueCurve {keys [<at 0 1>,<at 0.2174 1>,<at 0.4956 0.4798>,<at 0.797 1>,<at 1 1>]}
        GreenCurve {keys [<at 0 1>,<at 0.2174 1>,<at 0.4956 0.4798>,<at 0.797 1>,<at 1 1>]}
        RedCurve {keys [<at 0 0.2195>,<at 0.2174 1>,<at 0.4956 0.4798>,<at 0.797 1>,<at 1 0.2195>]}
       }}
      CurveIndex=&Frac_Float(Float=&Add_Float(
        A=&OneMinusFloat(Float=&ReturnNormalizedExecIndex(Normalized Index Scale=0.25))
        B=&Multiply_Float(A=$Engine.Time, B=0.3)
       ))
     ))
    ScaleA=true
    ScaleRGB=true
   ) {enabled=false}
   &DynamicMaterialParameters(
    Index 0 Param 1=&FloatFromCurve(
     FloatCurve=DataInterface() {propertyValues {Curve {
        defaultValue=1
        keys [
         {arriveTangent=2.0176, interpMode=RCIM_Cubic, leaveTangent=2.0176, tangentMode=RCTM_User, time=0.1, value=-1}
         {interpMode=RCIM_Cubic, tangentMode=RCTM_User, time=0.5, value=0}
         {arriveTangent=-3.0347, tangentMode=RCTM_User, time=0.9, value=-1}
        ]
       }}}
     CurveIndex=&Frac_Float(Float=&Add_Float(
       A=&OneMinusFloat(Float=&ReturnNormalizedExecIndex(Normalized Index Scale=0.5))
       B=&Multiply_Float(A=$Engine.Time, B=0.3)
      ))
    )
    Module.Index0Float0Write=true
    Module.Index0Float1Write=true
    Module.Index0Float2Write=true
    Module.Index0Float3Write=true
   )
  ]
  rendererClasses [&NiagaraSpriteRendererProperties, &NiagaraMeshRendererProperties]
  renderers [
   {rendererClass=&NiagaraSpriteRendererProperties}
   {rendererClass=&NiagaraMeshRendererProperties, rendererIndex=1}
  ]
 }
 Spiral_Particles_SpawnRate {
  bEnabled=false
  emitterUpdateScript [
   &EmitterState()
   &SpawnRate(SpawnRate=3, Use Spawn Probability=false)
   &SpawnBurst_Instantaneous(Spawn Count=35, Use Spawn Probability=false, UseLoopCountLimit=true) {enabled=false}
   /SetVariables(Emitter.Height=115, Emitter.SpiralFrequency=2, Emitter.SpiralSpeed=0.12) {isSetParameters=true, module=SetVariables_#9}
  ]
  particleSpawnScript [&InitializeParticle(
    Lifetime=&RandomRangeFloat(Minimum=10, Maximum=10)
    Mesh Scale Mode=::NewEnumerator1
    Mesh Uniform Scale Max=0.8
    Mesh Uniform Scale Min=0.5
    Sprite Size Mode=::NewEnumerator1
    Uniform Sprite Size Max=7
    UsePositionOffset=true
    Write Material Random=true
   )]
  particleUpdateScript [
   &ParticleState()
   /SetVariables(Particles.SpiralStrength:NiagaraFloat=&FloatFromCurve(
     FloatCurve=DataInterface() {propertyValues {Curve {
        defaultValue=1
        keys [
         {interpMode=RCIM_Cubic, leaveTangentWeight=0.2393, tangentMode=RCTM_User, time=0, value=1}
         {arriveTangent=-0.9582, interpMode=RCIM_Cubic, leaveTangent=-0.9582, tangentMode=RCTM_User, time=0.461, value=0.6565}
         <at 1 0.3093 RCIM_Cubic>
        ]
       }}}
     Scale Curve=&Waveform(
      Global Input Position=$Particles.NormalizedAge
      "Waveform [1]"=::NewEnumerator0
      "[1] Phase Scale"=0.25
      "[1] Frequency"=1
      "[1] Amplitude Min/Max"=<0.5 1>
      "[1] Scale"=17
     )
    )) {isSetParameters=true, module=SetVariables_#10}
   &ScaleSpriteSize(
    Uniform Curve Scale=&Waveform(
     Global Input Position=$Particles.NormalizedAge
     "Waveform [1]"=::NewEnumerator0
     "[1] Phase Scale"=0.5
     "[1] Frequency"=1
     "[1] Amplitude Min/Max"=<0.5 1>
    )
    Uniform Curve Sprite Scale=DataInterface() {propertyValues {Curve {
       defaultValue=0
       keys [
        {arriveTangent=4.5158, interpMode=RCIM_Cubic, leaveTangent=4.5158, tangentMode=RCTM_User, time=0, value=0.25}
        {arriveTangentWeight=0.1782, interpMode=RCIM_Cubic, leaveTangentWeight=0.1222, tangentMode=RCTM_User, time=0.2966, value=1}
        {arriveTangentWeight=0.0678, interpMode=RCIM_Cubic, leaveTangentWeight=0.0669, tangentMode=RCTM_User, time=0.5017, value=0.3444}
        {arriveTangentWeight=0.1206, interpMode=RCIM_Cubic, leaveTangentWeight=0.1799, tangentMode=RCTM_User, time=0.7007, value=1}
        {arriveTangent=-4.5159, tangentMode=RCTM_User, time=1, value=0.25}
       ]
      }}}
   )
   /SetVariables(Particles.PositionOffset:NiagaraPosition=&AddVectorToPosition(
     Position=&AddVectorToPosition(
      Position=<0 0 6>
      Vector=&Multiply_VectorByFloat(
       Vector=&RandomRangeVector(Minimum=<-1 -1 -1>, Maximum=<1 1 1>)
       Float=&FloatFromCurve(
        FloatCurve=DataInterface() {propertyValues {Curve {
           defaultValue=1
           keys [
            {interpMode=RCIM_Cubic, tangentMode=RCTM_User, time=0, value=0.1353}
            {arriveTangentWeight=0.3222, interpMode=RCIM_Cubic, leaveTangentWeight=0.0504, tangentMode=RCTM_User, time=0.25, value=1}
            {arriveTangentWeight=0.0226, interpMode=RCIM_Cubic, leaveTangentWeight=0.0234, tangentMode=RCTM_User, time=0.5, value=0.4235}
            {arriveTangentWeight=0.0523, interpMode=RCIM_Cubic, leaveTangentWeight=0.3204, tangentMode=RCTM_User, time=0.75, value=1}
            {tangentMode=RCTM_User, time=1, value=0.1353}
           ]
          }}}
        Scale Curve=&Waveform(
         Global Input Position=$Particles.NormalizedAge
         "Waveform [1]"=::NewEnumerator2
         "[1] Phase Scale"=0.1
         "[1] Frequency"=1.1
         "[1] Amplitude Min/Max"=<-1 1>
         "[1] Scale"=6
        )
       )
      )
     )
     Vector=&MakeVector(
      X=&Waveform(
       Global Input Position=$Particles.NormalizedAge
       Global Phase=&Multiply_Float(A=$Engine.Time, B=$Emitter.SpiralSpeed)
       "Waveform [1]"=::NewEnumerator0
       "[1] Frequency"=$Emitter.SpiralFrequency
       "[1] Amplitude Min/Max"=<-1 1>
       "[1] Scale"=$Particles.SpiralStrength
      )
      Y=&Waveform(
       Global Input Position=$Particles.NormalizedAge
       Global Phase=&Multiply_Float(A=$Engine.Time, B=$Emitter.SpiralSpeed)
       "Waveform [1]"=::NewEnumerator1
       "[1] Frequency"=$Emitter.SpiralFrequency
       "[1] Amplitude Min/Max"=<-1 1>
       "[1] Scale"=$Particles.SpiralStrength
      )
      Z=&FloatFromCurve(
       FloatCurve=DataInterface() {propertyValues {Curve {defaultValue=0, keys [{tangentMode=RCTM_User, time=0, value=0}, {tangentMode=RCTM_User, time=1, value=1}]}}}
       Scale Curve=$Emitter.Height
      )
     )
    )) {isSetParameters=true, module=SetVariables_#11}
   &ScaleMeshSize(Scale Factor=&Multiply_VectorByFloat(
     Vector=<1 1 1>
     Float=&Multiply_Float(
      A=1
      B=&FloatFromCurve(
       FloatCurve=DataInterface() {propertyValues {Curve {
          defaultValue=0
          keys [
           {arriveTangent=9.8295, interpMode=RCIM_Cubic, leaveTangent=9.8295, tangentMode=RCTM_User, time=0, value=0}
           {arriveTangent=4.7168, interpMode=RCIM_Cubic, leaveTangent=4.7168, tangentMode=RCTM_User, time=0.0304, value=0.5049}
           {arriveTangentWeight=0.1782, interpMode=RCIM_Cubic, leaveTangentWeight=0.1222, tangentMode=RCTM_User, time=0.2966, value=1}
           {arriveTangentWeight=0.0678, interpMode=RCIM_Cubic, leaveTangentWeight=0.0669, tangentMode=RCTM_User, time=0.5037, value=0.561}
           {arriveTangentWeight=0.1206, interpMode=RCIM_Cubic, leaveTangentWeight=0.1799, tangentMode=RCTM_User, time=0.7007, value=1}
           {arriveTangent=-4.1282, interpMode=RCIM_Cubic, leaveTangent=-4.1282, tangentMode=RCTM_User, time=0.9773, value=0.5146}
           {arriveTangent=-11.1163, tangentMode=RCTM_User, time=1, value=0}
          ]
         }}}
       Scale Curve=&Waveform(
        Global Input Position=$Particles.NormalizedAge
        "Waveform [1]"=::NewEnumerator0
        "[1] Phase Scale"=0.5
        "[1] Frequency"=1
        "[1] Amplitude Min/Max"=<0.7 1>
       )
      )
     )
    ))
   &ScaleColor(
    Scale RGB=&MakeVectorFromLinearColorRGB(Color=&ColorFromCurve(
      ColorCurve=DataInterface() {propertyValues {
        AlphaCurve {keys [<at 0 1>]}
        BlueCurve {keys [<at 0 1>,<at 0.2174 1>,<at 0.4956 0.4798>,<at 0.797 1>,<at 1 1>]}
        GreenCurve {keys [<at 0 1>,<at 0.2174 1>,<at 0.4956 0.4798>,<at 0.797 1>,<at 1 1>]}
        RedCurve {keys [<at 0 0.2195>,<at 0.2174 1>,<at 0.4956 0.4798>,<at 0.797 1>,<at 1 0.2195>]}
       }}
      CurveIndex=&Frac_Float(Float=&Add_Float(
        A=&OneMinusFloat(Float=&ReturnNormalizedExecIndex(Normalized Index Scale=0.25))
        B=&Multiply_Float(A=$Engine.Time, B=0.3)
       ))
     ))
    ScaleA=true
    ScaleRGB=true
   ) {enabled=false}
   &DynamicMaterialParameters(
    Index 0 Param 1=&FloatFromCurve(
     FloatCurve=DataInterface() {propertyValues {Curve {
        defaultValue=1
        keys [
         {arriveTangent=2.0176, interpMode=RCIM_Cubic, leaveTangent=2.0176, tangentMode=RCTM_User, time=0.1, value=-1}
         {interpMode=RCIM_Cubic, tangentMode=RCTM_User, time=0.5, value=0}
         {arriveTangent=-3.0347, tangentMode=RCTM_User, time=0.9, value=-1}
        ]
       }}}
     CurveIndex=&Frac_Float(Float=&Add_Float(A=$Particles.NormalizedAge, B=&Multiply_Float(A=$Engine.Time, B=0.3)))
    )
    Module.Index0Float0Write=true
    Module.Index0Float1Write=true
    Module.Index0Float2Write=true
    Module.Index0Float3Write=true
   )
  ]
  rendererClasses [&NiagaraSpriteRendererProperties, &NiagaraMeshRendererProperties]
  renderers [
   {rendererClass=&NiagaraSpriteRendererProperties}
   {rendererClass=&NiagaraMeshRendererProperties, rendererIndex=1}
  ]
 }
 Spiral_Particles_TravelingLoop {
  emitterUpdateScript [
   &EmitterState()
   &SpawnBurst_Instantaneous(Spawn Count=20, Use Spawn Probability=false, UseLoopCountLimit=true)
   /SetVariables(Emitter.Height=55, Emitter.SpiralFrequency=1, Emitter.TravelSpeed=0.12) {isSetParameters=true, module=SetVariables_#9}
  ]
  particleSpawnScript [
   &PartitionParticles(Number Of Partitions=4, Partition Mode:ENiagara_PartitionMode_v2=::NewEnumerator5)
   &InitializeParticle(
    Lifetime=5
    Mesh Scale Mode=::NewEnumerator1
    Mesh Uniform Scale Max=0.14
    Mesh Uniform Scale Min=0.09
    UsePositionOffset=false
    Write Material Random=true
   )
   &InitialMeshOrientation(
    Mesh Orientation Mode=::NewEnumerator2
    Rotation=&RandomRangeVector(Minimum=<0 -0.05 -0.05>, Maximum=<1 0.05 0.05>)
    Use Rotation Vector=true
   )
   /SetVariables(Particles.TravelSpeedParticle:NiagaraFloat=&FloatFromCurve(
     FloatCurve=DataInterface() {propertyValues {Curve {keys [<at 0 1>,<at 0.7931 1>,<at 1 0.7>]}}}
     CurveIndex=&RandomRangeFloat(Minimum=0, Maximum=1)
     Scale Curve=&MergeHelper_FloatSwitch(0=0.15, 1=0.1, 2=0.1, 3=0.1, Index=$Particles.PartitionParticles.Partition, Num=::NewEnumerator2)
    )) {isSetParameters=true, module=SetVariables_#12}
   &CameraOffset(Camera Offset Amount=3)
   &SkeletalMeshLocation(
    Bone / Socket Index=2
    MaskSpawnGroup=false
    Mesh Sampling Type=::NewEnumerator5
    Orientation Sampling=::NewEnumerator1
    Skeletal Mesh=$System.Skeletal Mesh
    Socket Sampling Mode=::NewEnumerator5
   )
  ]
  particleUpdateScript [
   &ParticleState(Kill Particles When Lifetime Has Elapsed=false, Let Infinitely Lived Particles Die When Emitter Deactivates=true)
   /SetVariables(
    Particles.CustomIndex:NiagaraFloat=&Frac_Float(Float=&Add_Float(
      A=&ReturnNormalizedExecIndex()
      B=&Multiply_Float(A=$Engine.Time, B=$Particles.TravelSpeedParticle)
     ))
    Particles.SpiralStrengthWaveform:NiagaraFloat=&Waveform(
     Global Input Position=&ReturnNormalizedExecIndex()
     "Waveform [1]"=::NewEnumerator0
     "[1] Phase Scale"=0.25
     "[1] Frequency"=1
     "[1] Amplitude Min/Max"=<1 1.1>
     "[1] Scale"=19.5
    )
   ) {isSetParameters=true, module=SetVariables_#10}
   /SetVariables(Particles.SpiralStrength:NiagaraFloat=&FloatFromCurve(
     FloatCurve=DataInterface() {propertyValues {Curve {
        defaultValue=1
        keys [
         {arriveTangent=-0.2299, interpMode=RCIM_Cubic, leaveTangent=-0.2299, leaveTangentWeight=0.2393, tangentMode=RCTM_User, time=0, value=0.5}
         {arriveTangent=-0.4468, interpMode=RCIM_Cubic, leaveTangent=-0.4468, tangentMode=RCTM_User, time=0.4032, value=0.4487}
         <at 1 0.35 RCIM_Cubic>
        ]
       }}}
     CurveIndex=$Particles.CustomIndex
     Scale Curve=$Particles.SpiralStrengthWaveform
    )) {isSetParameters=true, module=SetVariables_#13}
   /SetVariables(Particles.SpiralFrequency:NiagaraFloat=&FloatFromCurve(
     FloatCurve=DataInterface() {propertyValues {Curve {keys [<at 0 1>,<at 1 1.2707>]}}}
     CurveIndex=$Particles.CustomIndex
     Scale Curve=$Emitter.SpiralFrequency
    )) {isSetParameters=true, module=SetVariables_#14}
   /SetVariables(Particles.PositionOffset:NiagaraPosition=&AddVectorToPosition(
     Position=&AddVectorToPosition(
      Position=$Particles.Position
      Vector=&Multiply_VectorByFloat(
       Vector=&RandomRangeVector(Minimum=<-0.75 -0.75 -1>, Maximum=<0.75 0.75 1>)
       Float=&Add_Float(
        A=&FloatFromCurve(FloatCurve=DataInterface() {propertyValues {Curve {keys [<at 0 1>,<at 1 0.1>]}}}, CurveIndex=$Particles.CustomIndex, Scale Curve=4)
        B=&FloatFromCurve(
         FloatCurve=DataInterface() {propertyValues {Curve {
            defaultValue=1
            keys [
             {arriveTangent=5.6707, interpMode=RCIM_Cubic, leaveTangent=5.6707, tangentMode=RCTM_User, time=0, value=0.538}
             {arriveTangentWeight=0.3222, interpMode=RCIM_Cubic, leaveTangentWeight=0.0504, tangentMode=RCTM_User, time=0.1717, value=1.0736}
             {
              arriveTangent=-2.0781
              arriveTangentWeight=0.0226
              interpMode=RCIM_Cubic
              leaveTangent=-1.0421
              leaveTangentWeight=0.0234
              tangentMode=RCTM_Break
              time=0.59
              value=0.5528
             }
             {
              arriveTangent=-0.5117
              arriveTangentWeight=0.0523
              interpMode=RCIM_Cubic
              leaveTangent=-0.5117
              leaveTangentWeight=0.3204
              tangentMode=RCTM_User
              time=0.8408
              value=0.7639
             }
             {arriveTangent=-2.4831, interpMode=RCIM_Cubic, leaveTangent=-2.4831, tangentMode=RCTM_User, time=1, value=0.4}
            ]
           }}}
         CurveIndex=$Particles.CustomIndex
         Scale Curve=&Waveform(
          Global Input Position=&ReturnNormalizedExecIndex()
          "Waveform [1]"=::NewEnumerator2
          "[1] Phase Scale"=0.2
          "[1] Frequency"=1.1
          "[1] Amplitude Min/Max"=<-1 1>
          "[1] Scale"=&Lerp_Float(A=4, B=6, Alpha=$System.reactiveFloat)
         )
        )
       )
      )
     )
     Vector=&MakeVector(
      X=&Waveform(
       Global Input Position=$Particles.CustomIndex
       Global Phase=&Add_Float(A=&Multiply_Float(A=$Engine.Time, B=$Emitter.SpiralSpeed), B=$Emitter.SpiralOffset)
       "Waveform [1]"=::NewEnumerator0
       "[1] Frequency"=$Particles.SpiralFrequency
       "[1] Amplitude Min/Max"=<-1 1>
       "[1] Scale"=$Particles.SpiralStrength
      )
      Y=&Waveform(
       Global Input Position=$Particles.CustomIndex
       Global Phase=&Add_Float(A=&Multiply_Float(A=$Engine.Time, B=$Emitter.SpiralSpeed), B=$Emitter.SpiralOffset)
       "Waveform [1]"=::NewEnumerator1
       "[1] Frequency"=$Particles.SpiralFrequency
       "[1] Amplitude Min/Max"=<-1 1>
       "[1] Scale"=$Particles.SpiralStrength
      )
      Z=&FloatFromCurve(
       FloatCurve=DataInterface() {propertyValues {Curve {defaultValue=0, keys [{tangentMode=RCTM_User, time=0, value=0}, {tangentMode=RCTM_User, time=1, value=1}]}}}
       CurveIndex=$Particles.CustomIndex
       Scale Curve=&Multiply_Float(A=$Emitter.Height, B=1)
      )
     )
    )) {isSetParameters=true, module=SetVariables_#11}
   &ScaleMeshSize(Scale Factor=&Multiply_VectorByFloat(
     Vector=<1 1 1>
     Float=&Multiply_Float(
      A=&Multiply_Float(
       A=&MergeHelper_FloatSwitch(
        0=&Waveform(
         "Waveform [1]"=::NewEnumerator2
         "[1] Frequency"=1
         "[1] Amplitude Min/Max"=<-0.25 0.75>
         "[1] Scale"=$System.AudioLow
         "[1] Bias"=&RandomRangeFloat(Minimum=0.8, Maximum=1.1)
        )
        1=&Waveform(
         "Waveform [1]"=::NewEnumerator2
         "[1] Frequency"=1.5
         "[1] Amplitude Min/Max"=<-0.25 0.75>
         "[1] Scale"=$System.AudioMedium
         "[1] Bias"=&RandomRangeFloat(Minimum=0.8, Maximum=1.1)
        )
        2=&Waveform(
         "Waveform [1]"=::NewEnumerator2
         "[1] Frequency"=2
         "[1] Amplitude Min/Max"=<-0.25 0.75>
         "[1] Scale"=$System.AudioHigh
         "[1] Bias"=&RandomRangeFloat(Minimum=0.8, Maximum=1.1)
        )
        3=&Waveform(
         "Waveform [1]"=::NewEnumerator2
         "[1] Frequency"=1
         "[1] Amplitude Min/Max"=<-0.25 0.75>
         "[1] Scale"=$System.AudioLow
         "[1] Bias"=&RandomRangeFloat(Minimum=0.8, Maximum=1.1)
        )
        Index=$Particles.PartitionParticles.Partition
        Num=::NewEnumerator2
       )
       B=1
      )
      B=&FloatFromCurve(
       FloatCurve=DataInterface() {propertyValues {Curve {
          defaultValue=0
          keys [
           {arriveTangent=14.706, interpMode=RCIM_Cubic, leaveTangent=14.706, leaveTangentWeight=0.0825, tangentMode=RCTM_User, time=0, value=0}
           {
            arriveTangent=1.3013
            arriveTangentWeight=0.1782
            interpMode=RCIM_Cubic
            leaveTangent=1.3013
            leaveTangentWeight=0.1222
            tangentMode=RCTM_User
            time=0.1323
            value=0.8
           }
           {
            arriveTangent=-2.2321
            arriveTangentWeight=0.0723
            interpMode=RCIM_Cubic
            leaveTangent=-2.2321
            leaveTangentWeight=0.0438
            tangentMode=RCTM_User
            time=0.2752
            value=0.6597
           }
           {
            arriveTangent=-0.6475
            arriveTangentWeight=0.1638
            interpMode=RCIM_Cubic
            leaveTangent=-0.6475
            leaveTangentWeight=0.2008
            tangentMode=RCTM_User
            time=0.401
            value=0.7
           }
           {
            arriveTangent=0.1355
            arriveTangentWeight=0.0769
            interpMode=RCIM_Cubic
            leaveTangent=0.1355
            leaveTangentWeight=0.054
            tangentMode=RCTM_User
            time=0.8622
            value=0.6
           }
           {arriveTangent=-12.6776, interpMode=RCIM_Cubic, leaveTangent=-12.6776, tangentMode=RCTM_User, time=1, value=0}
          ]
         }}}
       CurveIndex=$Particles.CustomIndex
      )
     )
    )) {module=ScaleMeshSize001}
   &ScaleColor(
    Scale RGB=&MakeVectorFromLinearColorRGB(Color=&MergeHelper_ColorSwitch(
      0=&LerpLinearColors(
       StartColor=<rgba 1 1 1 1>
       EndColor=&RandomRangeLinearColor(Minimum=<rgba 1 0.4 0.4 1>, Maximum=<rgba 0.4 1 0.4 1>)
       LerpFactor=&Waveform("Waveform [1]"=::NewEnumerator2, "[1] Frequency"=1, "[1] Amplitude Min/Max"=<0 1.2>, "[1] Scale"=$System.AudioLow)
      )
      1=&LerpLinearColors(
       StartColor=<rgba 1 1 1 1>
       EndColor=&RandomRangeLinearColor(Minimum=<rgba 0.4 1 0.4 1>, Maximum=<rgba 0.4 0.4 1 1>)
       LerpFactor=&Waveform("Waveform [1]"=::NewEnumerator2, "[1] Frequency"=1.5, "[1] Amplitude Min/Max"=<0 1.2>, "[1] Scale"=$System.AudioMedium)
      )
      2=&LerpLinearColors(
       StartColor=<rgba 1 1 1 1>
       EndColor=&RandomRangeLinearColor(Minimum=<rgba 0.4 0.4 1 1>, Maximum=<rgba 1 0.4 0.4 1>)
       LerpFactor=&Waveform("Waveform [1]"=::NewEnumerator2, "[1] Frequency"=2, "[1] Amplitude Min/Max"=<0 1.2>, "[1] Scale"=$System.AudioHigh)
      )
      3=&LerpLinearColors(
       StartColor=<rgba 1 1 1 1>
       EndColor=&RandomRangeLinearColor(Minimum=<rgba 1 0.4 0.4 1>, Maximum=<rgba 0.4 1 0.4 1>)
       LerpFactor=&Waveform("Waveform [1]"=::NewEnumerator2, "[1] Frequency"=1, "[1] Amplitude Min/Max"=<0 1.2>, "[1] Scale"=$System.AudioLow)
      )
      Index=$Particles.PartitionParticles.Partition
      Num=::NewEnumerator2
     ))
    ScaleA=true
    ScaleRGB=true
   ) {module=ScaleColor001}
   &ScaleMeshSize(Scale Factor=&MakeVector(
     X=&Waveform("Waveform [1]"=::NewEnumerator2, "[1] Frequency"=1, "[1] Amplitude Min/Max"=<-0.25 0.25>, "[1] Scale"=$System.AudioLow, "[1] Bias"=1)
     Y=&Waveform("Waveform [1]"=::NewEnumerator2, "[1] Frequency"=2, "[1] Amplitude Min/Max"=<-0.25 0.25>, "[1] Scale"=$System.AudioMedium, "[1] Bias"=1)
     Z=&Waveform("Waveform [1]"=::NewEnumerator2, "[1] Frequency"=3, "[1] Amplitude Min/Max"=<-0.25 0.25>, "[1] Scale"=$System.AudioHigh, "[1] Bias"=1)
    ))
   &ScaleMeshSize(Scale Factor=&VectorFromFloat(Value=&Lerp_Float(
      A=&Lerp_Float(A=$System.FrontEndReactiveScale, B=1, Alpha=$System.Timeline.NormalizedTime)
      B=&FN_FrontEndFloatModifier(Gameplay=1, Front End=1.1)
      Alpha=$System.reactiveFloat
     ))) {module=ScaleMeshSize002}
   &DynamicMaterialParameters(
    Index 0 Param 1=$System.EmissiveIntensity
    Index 0 Param 4=&MergeHelper_FloatSwitch(
     0=&Waveform("Waveform [1]"=::NewEnumerator2, "[1] Frequency"=1, "[1] Amplitude Min/Max"=<0 1>, "[1] Scale"=$System.AudioLow)
     1=&Waveform("Waveform [1]"=::NewEnumerator2, "[1] Frequency"=1.5, "[1] Amplitude Min/Max"=<0 1>, "[1] Scale"=$System.AudioMedium)
     2=&Waveform("Waveform [1]"=::NewEnumerator2, "[1] Frequency"=2, "[1] Amplitude Min/Max"=<0 1>, "[1] Scale"=$System.AudioHigh)
     3=&Waveform("Waveform [1]"=::NewEnumerator2, "[1] Frequency"=1, "[1] Amplitude Min/Max"=<0 1>, "[1] Scale"=$System.AudioLow)
     Index=$Particles.PartitionParticles.Partition
     Num=::NewEnumerator2
    )
    Module.Index0Float0Write=true
    Module.Index0Float1Write=true
    Module.Index0Float2Write=true
    Module.Index0Float3Write=true
   ) {module=DynamicMaterialParameters001}
  ]
  rendererClasses [&NiagaraMeshRendererProperties]
  renderers [{rendererClass=&NiagaraMeshRendererProperties, rendererIndex=0}]
 }
 Spiral_Particles_TravelingLoop001 {
  emitterUpdateScript [
   &EmitterState()
   &SpawnBurst_Instantaneous(Spawn Count=50, Use Spawn Probability=false, UseLoopCountLimit=true)
   /SetVariables(Emitter.Height=90, Emitter.SpiralFrequency=2, Emitter.TravelSpeed=0.12) {isSetParameters=true, module=SetVariables_#9}
  ]
  particleSpawnScript [
   &PartitionParticles(Number Of Partitions=4, Partition Mode:ENiagara_PartitionMode_v2=::NewEnumerator5)
   &InitializeParticle(
    Lifetime=5
    Mesh Scale Mode=::NewEnumerator1
    Mesh Uniform Scale Max=0.14
    Mesh Uniform Scale Min=0.09
    UsePositionOffset=false
    Write Material Random=true
   )
   &InitialMeshOrientation(
    Mesh Orientation Mode=::NewEnumerator2
    Rotation=&RandomRangeVector(Minimum=<0 -0.05 -0.05>, Maximum=<1 0.05 0.05>)
    Use Rotation Vector=true
   )
   /SetVariables(Particles.TravelSpeedParticle:NiagaraFloat=&FloatFromCurve(
     FloatCurve=DataInterface() {propertyValues {Curve {keys [<at 0 1>,<at 0.7931 1>,<at 1 0.7>]}}}
     CurveIndex=&RandomRangeFloat(Minimum=0, Maximum=1)
     Scale Curve=&MergeHelper_FloatSwitch(0=0.15, 1=0.1, 2=0.1, 3=0.1, Index=$Particles.PartitionParticles.Partition, Num=::NewEnumerator2)
    )) {isSetParameters=true, module=SetVariables_#12}
   &CameraOffset(Camera Offset Amount=3)
   &SkeletalMeshLocation(
    Bone / Socket Index=3
    MaskSpawnGroup=false
    Mesh Sampling Type=::NewEnumerator5
    Orientation Sampling=::NewEnumerator1
    Skeletal Mesh=$System.Skeletal Mesh
    Socket Sampling Mode=::NewEnumerator5
   )
  ]
  particleUpdateScript [
   &ParticleState(Kill Particles When Lifetime Has Elapsed=false, Let Infinitely Lived Particles Die When Emitter Deactivates=true)
   /SetVariables(
    Particles.CustomIndex:NiagaraFloat=&Frac_Float(Float=&Add_Float(
      A=&ReturnNormalizedExecIndex()
      B=&Multiply_Float(A=$Engine.Time, B=$Particles.TravelSpeedParticle)
     ))
    Particles.SpiralStrengthWaveform:NiagaraFloat=&Waveform(
     Global Input Position=&ReturnNormalizedExecIndex()
     "Waveform [1]"=::NewEnumerator0
     "[1] Phase Scale"=0.25
     "[1] Frequency"=1
     "[1] Amplitude Min/Max"=<1 1.1>
     "[1] Scale"=19.5
    )
   ) {isSetParameters=true, module=SetVariables_#10}
   /SetVariables(Particles.SpiralStrength:NiagaraFloat=&FloatFromCurve(
     FloatCurve=DataInterface() {propertyValues {Curve {
        defaultValue=1
        keys [<at 0 0.6 RCIM_Cubic>, {arriveTangent=-0.4468, interpMode=RCIM_Cubic, leaveTangent=-0.4468, tangentMode=RCTM_User, time=0.4032, value=0.3}]
       }}}
     CurveIndex=$Particles.CustomIndex
     Scale Curve=$Particles.SpiralStrengthWaveform
    )) {isSetParameters=true, module=SetVariables_#13}
   /SetVariables(Particles.SpiralFrequency:NiagaraFloat=&FloatFromCurve(
     FloatCurve=DataInterface() {propertyValues {Curve {keys [<at 0 1>,<at 1 1.2707>]}}}
     CurveIndex=$Particles.CustomIndex
     Scale Curve=$Emitter.SpiralFrequency
    )) {isSetParameters=true, module=SetVariables_#14}
   /SetVariables(Particles.PositionOffset:NiagaraPosition=&AddVectorToPosition(
     Position=&AddVectorToPosition(
      Position=$Particles.Position
      Vector=&Multiply_VectorByFloat(
       Vector=&RandomRangeVector(Minimum=<-0.75 -0.75 -1>, Maximum=<0.75 0.75 1>)
       Float=&Add_Float(
        A=&FloatFromCurve(FloatCurve=DataInterface() {propertyValues {Curve {keys [<at 0 1>,<at 1 0.1>]}}}, CurveIndex=$Particles.CustomIndex, Scale Curve=4)
        B=&FloatFromCurve(
         FloatCurve=DataInterface() {propertyValues {Curve {
            defaultValue=1
            keys [
             {arriveTangent=5.6707, interpMode=RCIM_Cubic, leaveTangent=5.6707, tangentMode=RCTM_User, time=0, value=0.538}
             {arriveTangentWeight=0.3222, interpMode=RCIM_Cubic, leaveTangentWeight=0.0504, tangentMode=RCTM_User, time=0.1717, value=1.0736}
             {
              arriveTangent=-2.0781
              arriveTangentWeight=0.0226
              interpMode=RCIM_Cubic
              leaveTangent=-1.0421
              leaveTangentWeight=0.0234
              tangentMode=RCTM_Break
              time=0.59
              value=0.5528
             }
             {
              arriveTangent=-0.5117
              arriveTangentWeight=0.0523
              interpMode=RCIM_Cubic
              leaveTangent=-0.5117
              leaveTangentWeight=0.3204
              tangentMode=RCTM_User
              time=0.8408
              value=0.7639
             }
             {arriveTangent=-2.4831, interpMode=RCIM_Cubic, leaveTangent=-2.4831, tangentMode=RCTM_User, time=1, value=0.4}
            ]
           }}}
         CurveIndex=$Particles.CustomIndex
         Scale Curve=&Waveform(
          Global Input Position=&ReturnNormalizedExecIndex()
          "Waveform [1]"=::NewEnumerator2
          "[1] Phase Scale"=0.2
          "[1] Frequency"=1.1
          "[1] Amplitude Min/Max"=<-1 1>
          "[1] Scale"=&Lerp_Float(A=4, B=6, Alpha=$System.reactiveFloat)
         )
        )
       )
      )
     )
     Vector=&MakeVector(
      X=&Waveform(
       Global Input Position=$Particles.CustomIndex
       Global Phase=&Add_Float(A=&Multiply_Float(A=$Engine.Time, B=$Emitter.SpiralSpeed), B=$Emitter.SpiralOffset)
       "Waveform [1]"=::NewEnumerator0
       "[1] Frequency"=$Particles.SpiralFrequency
       "[1] Amplitude Min/Max"=<-1 1>
       "[1] Scale"=$Particles.SpiralStrength
      )
      Y=&Waveform(
       Global Input Position=$Particles.CustomIndex
       Global Phase=&Add_Float(A=&Multiply_Float(A=$Engine.Time, B=$Emitter.SpiralSpeed), B=$Emitter.SpiralOffset)
       "Waveform [1]"=::NewEnumerator1
       "[1] Frequency"=$Particles.SpiralFrequency
       "[1] Amplitude Min/Max"=<-1 1>
       "[1] Scale"=$Particles.SpiralStrength
      )
      Z=&FloatFromCurve(
       FloatCurve=DataInterface() {propertyValues {Curve {defaultValue=0, keys [{tangentMode=RCTM_User, time=0, value=0}, {tangentMode=RCTM_User, time=1, value=1}]}}}
       CurveIndex=$Particles.CustomIndex
       Scale Curve=&Multiply_Float(A=$Emitter.Height, B=1)
      )
     )
    )) {isSetParameters=true, module=SetVariables_#11}
   &ScaleMeshSize(Scale Factor=&Multiply_VectorByFloat(
     Vector=<1 1 1>
     Float=&Multiply_Float(
      A=&Multiply_Float(
       A=&MergeHelper_FloatSwitch(
        0=&Waveform(
         "Waveform [1]"=::NewEnumerator2
         "[1] Frequency"=1
         "[1] Amplitude Min/Max"=<-0.25 0.75>
         "[1] Scale"=$System.AudioLow
         "[1] Bias"=&RandomRangeFloat(Minimum=0.8, Maximum=1.1)
        )
        1=&Waveform(
         "Waveform [1]"=::NewEnumerator2
         "[1] Frequency"=1.5
         "[1] Amplitude Min/Max"=<-0.25 0.75>
         "[1] Scale"=$System.AudioMedium
         "[1] Bias"=&RandomRangeFloat(Minimum=0.8, Maximum=1.1)
        )
        2=&Waveform(
         "Waveform [1]"=::NewEnumerator2
         "[1] Frequency"=2
         "[1] Amplitude Min/Max"=<-0.25 0.75>
         "[1] Scale"=$System.AudioHigh
         "[1] Bias"=&RandomRangeFloat(Minimum=0.8, Maximum=1.1)
        )
        3=&Waveform(
         "Waveform [1]"=::NewEnumerator2
         "[1] Frequency"=1
         "[1] Amplitude Min/Max"=<-0.25 0.75>
         "[1] Scale"=$System.AudioLow
         "[1] Bias"=&RandomRangeFloat(Minimum=0.8, Maximum=1.1)
        )
        Index=$Particles.PartitionParticles.Partition
        Num=::NewEnumerator2
       )
       B=1
      )
      B=&FloatFromCurve(
       FloatCurve=DataInterface() {propertyValues {Curve {
          defaultValue=0
          keys [
           {arriveTangent=14.706, interpMode=RCIM_Cubic, leaveTangent=14.706, leaveTangentWeight=0.0825, tangentMode=RCTM_User, time=0, value=0}
           {
            arriveTangent=1.3013
            arriveTangentWeight=0.1782
            interpMode=RCIM_Cubic
            leaveTangent=1.3013
            leaveTangentWeight=0.1222
            tangentMode=RCTM_User
            time=0.1323
            value=0.8
           }
           {
            arriveTangent=-2.2321
            arriveTangentWeight=0.0723
            interpMode=RCIM_Cubic
            leaveTangent=-2.2321
            leaveTangentWeight=0.0438
            tangentMode=RCTM_User
            time=0.2752
            value=0.6597
           }
           {
            arriveTangent=-0.6475
            arriveTangentWeight=0.1638
            interpMode=RCIM_Cubic
            leaveTangent=-0.6475
            leaveTangentWeight=0.2008
            tangentMode=RCTM_User
            time=0.401
            value=0.7
           }
           {
            arriveTangent=-9.6816
            arriveTangentWeight=0.1494
            interpMode=RCIM_Cubic
            leaveTangentWeight=0.0424
            tangentMode=RCTM_Break
            time=0.7
            value=0
           }
           {
            arriveTangentWeight=0.0424
            interpMode=RCIM_Cubic
            leaveTangent=12.3539
            leaveTangentWeight=0.0531
            tangentMode=RCTM_Break
            time=0.75
            value=0
           }
           {
            arriveTangent=0.1355
            arriveTangentWeight=0.0769
            interpMode=RCIM_Cubic
            leaveTangent=0.1355
            leaveTangentWeight=0.054
            tangentMode=RCTM_User
            time=0.8622
            value=0.6
           }
           {arriveTangent=-12.6776, interpMode=RCIM_Cubic, leaveTangent=-12.6776, tangentMode=RCTM_User, time=1, value=0}
          ]
         }}}
       CurveIndex=$Particles.CustomIndex
      )
     )
    )) {module=ScaleMeshSize001}
   &ScaleColor(
    Scale RGB=&MakeVectorFromLinearColorRGB(Color=&MergeHelper_ColorSwitch(
      0=&LerpLinearColors(
       StartColor=<rgba 1 1 1 1>
       EndColor=&RandomRangeLinearColor(Minimum=<rgba 1 0.4 0.4 1>, Maximum=<rgba 0.4 1 0.4 1>)
       LerpFactor=&Waveform("Waveform [1]"=::NewEnumerator2, "[1] Frequency"=1, "[1] Amplitude Min/Max"=<0 1.2>, "[1] Scale"=$System.AudioLow)
      )
      1=&LerpLinearColors(
       StartColor=<rgba 1 1 1 1>
       EndColor=&RandomRangeLinearColor(Minimum=<rgba 0.4 1 0.4 1>, Maximum=<rgba 0.4 0.4 1 1>)
       LerpFactor=&Waveform("Waveform [1]"=::NewEnumerator2, "[1] Frequency"=1.5, "[1] Amplitude Min/Max"=<0 1.2>, "[1] Scale"=$System.AudioMedium)
      )
      2=&LerpLinearColors(
       StartColor=<rgba 1 1 1 1>
       EndColor=&RandomRangeLinearColor(Minimum=<rgba 0.4 0.4 1 1>, Maximum=<rgba 1 0.4 0.4 1>)
       LerpFactor=&Waveform("Waveform [1]"=::NewEnumerator2, "[1] Frequency"=2, "[1] Amplitude Min/Max"=<0 1.2>, "[1] Scale"=$System.AudioHigh)
      )
      3=&LerpLinearColors(
       StartColor=<rgba 1 1 1 1>
       EndColor=&RandomRangeLinearColor(Minimum=<rgba 1 0.4 0.4 1>, Maximum=<rgba 0.4 1 0.4 1>)
       LerpFactor=&Waveform("Waveform [1]"=::NewEnumerator2, "[1] Frequency"=1, "[1] Amplitude Min/Max"=<0 1.2>, "[1] Scale"=$System.AudioLow)
      )
      Index=$Particles.PartitionParticles.Partition
      Num=::NewEnumerator2
     ))
    ScaleA=true
    ScaleRGB=true
   ) {module=ScaleColor001}
   &ScaleMeshSize(Scale Factor=&MakeVector(
     X=&Waveform("Waveform [1]"=::NewEnumerator2, "[1] Frequency"=1, "[1] Amplitude Min/Max"=<-0.25 0.25>, "[1] Scale"=$System.AudioLow, "[1] Bias"=1)
     Y=&Waveform("Waveform [1]"=::NewEnumerator2, "[1] Frequency"=2, "[1] Amplitude Min/Max"=<-0.25 0.25>, "[1] Scale"=$System.AudioMedium, "[1] Bias"=1)
     Z=&Waveform("Waveform [1]"=::NewEnumerator2, "[1] Frequency"=3, "[1] Amplitude Min/Max"=<-0.25 0.25>, "[1] Scale"=$System.AudioHigh, "[1] Bias"=1)
    ))
   &ScaleMeshSize(Scale Factor=&VectorFromFloat(Value=&Lerp_Float(
      A=&Lerp_Float(A=$System.FrontEndReactiveScale, B=1, Alpha=$System.Timeline.NormalizedTime)
      B=&FN_FrontEndFloatModifier(Gameplay=1, Front End=1.1)
      Alpha=$System.reactiveFloat
     ))) {module=ScaleMeshSize002}
   &DynamicMaterialParameters(
    Index 0 Param 1=$System.EmissiveIntensity
    Index 0 Param 4=&MergeHelper_FloatSwitch(
     0=&Waveform("Waveform [1]"=::NewEnumerator2, "[1] Frequency"=1, "[1] Amplitude Min/Max"=<0 1>, "[1] Scale"=$System.AudioLow)
     1=&Waveform("Waveform [1]"=::NewEnumerator2, "[1] Frequency"=1.5, "[1] Amplitude Min/Max"=<0 1>, "[1] Scale"=$System.AudioMedium)
     2=&Waveform("Waveform [1]"=::NewEnumerator2, "[1] Frequency"=2, "[1] Amplitude Min/Max"=<0 1>, "[1] Scale"=$System.AudioHigh)
     3=&Waveform("Waveform [1]"=::NewEnumerator2, "[1] Frequency"=1, "[1] Amplitude Min/Max"=<0 1>, "[1] Scale"=$System.AudioLow)
     Index=$Particles.PartitionParticles.Partition
     Num=::NewEnumerator2
    )
    Module.Index0Float0Write=true
    Module.Index0Float1Write=true
    Module.Index0Float2Write=true
    Module.Index0Float3Write=true
   ) {module=DynamicMaterialParameters001}
  ]
  rendererClasses [&NiagaraMeshRendererProperties]
  renderers [{rendererClass=&NiagaraMeshRendererProperties, rendererIndex=0}]
 }
 Spiral_Particles_TravelingLoop002 {
  emitterUpdateScript [
   &EmitterState()
   &SpawnBurst_Instantaneous(Spawn Count=60, Use Spawn Probability=false, UseLoopCountLimit=true)
   /SetVariables(Emitter.Height=90, Emitter.SpiralFrequency=2.5, Emitter.TravelSpeed=0.12) {isSetParameters=true, module=SetVariables_#9}
  ]
  particleSpawnScript [
   &PartitionParticles(Number Of Partitions=4, Partition Mode:ENiagara_PartitionMode_v2=::NewEnumerator5)
   &InitializeParticle(
    Lifetime=5
    Mesh Scale Mode=::NewEnumerator1
    Mesh Uniform Scale Max=0.14
    Mesh Uniform Scale Min=0.09
    UsePositionOffset=false
    Write Material Random=true
   )
   &InitialMeshOrientation(
    Mesh Orientation Mode=::NewEnumerator2
    Rotation=&RandomRangeVector(Minimum=<0 -0.05 -0.05>, Maximum=<1 0.05 0.05>)
    Use Rotation Vector=true
   )
   /SetVariables(Particles.TravelSpeedParticle:NiagaraFloat=&FloatFromCurve(
     FloatCurve=DataInterface() {propertyValues {Curve {keys [<at 0.7931 0.9>,<at 1 0.7>]}}}
     CurveIndex=&RandomRangeFloat(Minimum=0, Maximum=1)
     Scale Curve=&MergeHelper_FloatSwitch(0=0.15, 1=0.1, 2=0.1, 3=0.1, Index=$Particles.PartitionParticles.Partition, Num=::NewEnumerator2)
    )) {isSetParameters=true, module=SetVariables_#12}
   &CameraOffset(Camera Offset Amount=3)
   &SkeletalMeshLocation(
    MaskSpawnGroup=false
    Mesh Sampling Type=::NewEnumerator5
    Orientation Sampling=::NewEnumerator1
    Skeletal Mesh=$System.Skeletal Mesh
    Socket Sampling Mode=::NewEnumerator5
   )
  ]
  particleUpdateScript [
   &ParticleState(Kill Particles When Lifetime Has Elapsed=false, Let Infinitely Lived Particles Die When Emitter Deactivates=true)
   /SetVariables(
    Particles.CustomIndex:NiagaraFloat=&Frac_Float(Float=&Add_Float(
      A=&ReturnNormalizedExecIndex()
      B=&Multiply_Float(A=$Engine.Time, B=$Particles.TravelSpeedParticle)
     ))
    Particles.SpiralStrengthWaveform:NiagaraFloat=&Waveform(
     Global Input Position=&ReturnNormalizedExecIndex()
     "Waveform [1]"=::NewEnumerator0
     "[1] Phase Scale"=0.25
     "[1] Frequency"=1
     "[1] Amplitude Min/Max"=<1 1.1>
     "[1] Scale"=19.5
    )
   ) {isSetParameters=true, module=SetVariables_#10}
   /SetVariables(Particles.SpiralStrength:NiagaraFloat=&FloatFromCurve(
     FloatCurve=DataInterface() {propertyValues {Curve {
        defaultValue=1
        keys [
         {arriveTangent=-0.7999, interpMode=RCIM_Cubic, leaveTangent=-0.7999, leaveTangentWeight=0.2393, tangentMode=RCTM_User, time=0, value=0.45}
         {arriveTangent=-0.4468, interpMode=RCIM_Cubic, leaveTangent=-0.4468, tangentMode=RCTM_User, time=0.4032, value=0.4487}
         <at 1 0.35 RCIM_Cubic>
        ]
       }}}
     CurveIndex=$Particles.CustomIndex
     Scale Curve=$Particles.SpiralStrengthWaveform
    )) {isSetParameters=true, module=SetVariables_#13}
   /SetVariables(Particles.SpiralFrequency:NiagaraFloat=&FloatFromCurve(
     FloatCurve=DataInterface() {propertyValues {Curve {keys [<at 0 1>,<at 1 1.2707>]}}}
     CurveIndex=$Particles.CustomIndex
     Scale Curve=$Emitter.SpiralFrequency
    )) {isSetParameters=true, module=SetVariables_#14}
   /SetVariables(Particles.PositionOffset:NiagaraPosition=&AddVectorToPosition(
     Position=&AddVectorToPosition(
      Position=$Particles.Position
      Vector=&Multiply_VectorByFloat(
       Vector=&RandomRangeVector(Minimum=<-0.75 -0.75 -1>, Maximum=<0.75 0.75 1>)
       Float=&Add_Float(
        A=&FloatFromCurve(FloatCurve=DataInterface() {propertyValues {Curve {keys [<at 0 1>,<at 1 0.1>]}}}, CurveIndex=$Particles.CustomIndex, Scale Curve=4)
        B=&FloatFromCurve(
         FloatCurve=DataInterface() {propertyValues {Curve {
            defaultValue=1
            keys [
             {arriveTangent=5.6707, interpMode=RCIM_Cubic, leaveTangent=5.6707, tangentMode=RCTM_User, time=0, value=0.538}
             {arriveTangentWeight=0.3222, interpMode=RCIM_Cubic, leaveTangentWeight=0.0504, tangentMode=RCTM_User, time=0.1717, value=1.0736}
             {
              arriveTangent=-2.0781
              arriveTangentWeight=0.0226
              interpMode=RCIM_Cubic
              leaveTangent=-1.0421
              leaveTangentWeight=0.0234
              tangentMode=RCTM_Break
              time=0.59
              value=0.5528
             }
             {
              arriveTangent=-0.5117
              arriveTangentWeight=0.0523
              interpMode=RCIM_Cubic
              leaveTangent=-0.5117
              leaveTangentWeight=0.3204
              tangentMode=RCTM_User
              time=0.8408
              value=0.7639
             }
             {arriveTangent=-2.4831, interpMode=RCIM_Cubic, leaveTangent=-2.4831, tangentMode=RCTM_User, time=1, value=0.4}
            ]
           }}}
         CurveIndex=$Particles.CustomIndex
         Scale Curve=&Waveform(
          Global Input Position=&ReturnNormalizedExecIndex()
          "Waveform [1]"=::NewEnumerator2
          "[1] Phase Scale"=0.2
          "[1] Frequency"=1.1
          "[1] Amplitude Min/Max"=<-1 1>
          "[1] Scale"=&Lerp_Float(A=4, B=6, Alpha=$System.reactiveFloat)
         )
        )
       )
      )
     )
     Vector=&MakeVector(
      X=&Waveform(
       Global Input Position=$Particles.CustomIndex
       Global Phase=&Add_Float(A=&Multiply_Float(A=$Engine.Time, B=$Emitter.SpiralSpeed), B=$Emitter.SpiralOffset)
       "Waveform [1]"=::NewEnumerator0
       "[1] Frequency"=$Particles.SpiralFrequency
       "[1] Amplitude Min/Max"=<-1 1>
       "[1] Scale"=$Particles.SpiralStrength
      )
      Y=&Waveform(
       Global Input Position=$Particles.CustomIndex
       Global Phase=&Add_Float(A=&Multiply_Float(A=$Engine.Time, B=$Emitter.SpiralSpeed), B=$Emitter.SpiralOffset)
       "Waveform [1]"=::NewEnumerator1
       "[1] Frequency"=$Particles.SpiralFrequency
       "[1] Amplitude Min/Max"=<-1 1>
       "[1] Scale"=$Particles.SpiralStrength
      )
      Z=&FloatFromCurve(
       FloatCurve=DataInterface() {propertyValues {Curve {defaultValue=0, keys [{tangentMode=RCTM_User, time=0, value=0}, {tangentMode=RCTM_User, time=1, value=1}]}}}
       CurveIndex=$Particles.CustomIndex
       Scale Curve=&Multiply_Float(A=$Emitter.Height, B=1)
      )
     )
    )) {isSetParameters=true, module=SetVariables_#11}
   &ScaleMeshSize(Scale Factor=&Multiply_VectorByFloat(
     Vector=<1 1 1>
     Float=&Multiply_Float(
      A=&Multiply_Float(
       A=&MergeHelper_FloatSwitch(
        0=&Waveform(
         "Waveform [1]"=::NewEnumerator2
         "[1] Frequency"=1
         "[1] Amplitude Min/Max"=<-0.25 0.75>
         "[1] Scale"=$System.AudioLow
         "[1] Bias"=&RandomRangeFloat(Minimum=0.8, Maximum=1.1)
        )
        1=&Waveform(
         "Waveform [1]"=::NewEnumerator2
         "[1] Frequency"=1.5
         "[1] Amplitude Min/Max"=<-0.25 0.75>
         "[1] Scale"=$System.AudioMedium
         "[1] Bias"=&RandomRangeFloat(Minimum=0.8, Maximum=1.1)
        )
        2=&Waveform(
         "Waveform [1]"=::NewEnumerator2
         "[1] Frequency"=2
         "[1] Amplitude Min/Max"=<-0.25 0.75>
         "[1] Scale"=$System.AudioHigh
         "[1] Bias"=&RandomRangeFloat(Minimum=0.8, Maximum=1.1)
        )
        3=&Waveform(
         "Waveform [1]"=::NewEnumerator2
         "[1] Frequency"=1
         "[1] Amplitude Min/Max"=<-0.25 0.75>
         "[1] Scale"=$System.AudioLow
         "[1] Bias"=&RandomRangeFloat(Minimum=0.8, Maximum=1.1)
        )
        Index=$Particles.PartitionParticles.Partition
        Num=::NewEnumerator2
       )
       B=1
      )
      B=&FloatFromCurve(
       FloatCurve=DataInterface() {propertyValues {Curve {
          defaultValue=0
          keys [
           {arriveTangent=14.706, interpMode=RCIM_Cubic, leaveTangent=14.706, leaveTangentWeight=0.0825, tangentMode=RCTM_User, time=0, value=0}
           {
            arriveTangent=1.3013
            arriveTangentWeight=0.1782
            interpMode=RCIM_Cubic
            leaveTangent=1.3013
            leaveTangentWeight=0.1222
            tangentMode=RCTM_User
            time=0.1323
            value=0.8
           }
           {
            arriveTangent=-2.2321
            arriveTangentWeight=0.0723
            interpMode=RCIM_Cubic
            leaveTangent=-2.2321
            leaveTangentWeight=0.0438
            tangentMode=RCTM_User
            time=0.2752
            value=0.6597
           }
           {
            arriveTangent=-0.6475
            arriveTangentWeight=0.1638
            interpMode=RCIM_Cubic
            leaveTangent=-0.6475
            leaveTangentWeight=0.2008
            tangentMode=RCTM_User
            time=0.401
            value=0.7
           }
           {
            arriveTangent=-9.6816
            arriveTangentWeight=0.1494
            interpMode=RCIM_Cubic
            leaveTangentWeight=0.0424
            tangentMode=RCTM_Break
            time=0.7
            value=0
           }
           {
            arriveTangentWeight=0.0424
            interpMode=RCIM_Cubic
            leaveTangent=12.3539
            leaveTangentWeight=0.0531
            tangentMode=RCTM_Break
            time=0.75
            value=0
           }
           {
            arriveTangent=0.1355
            arriveTangentWeight=0.0769
            interpMode=RCIM_Cubic
            leaveTangent=0.1355
            leaveTangentWeight=0.054
            tangentMode=RCTM_User
            time=0.8622
            value=0.6
           }
           {arriveTangent=-12.6776, interpMode=RCIM_Cubic, leaveTangent=-12.6776, tangentMode=RCTM_User, time=1, value=0}
          ]
         }}}
       CurveIndex=$Particles.CustomIndex
      )
     )
    )) {module=ScaleMeshSize001}
   &ScaleColor(
    Scale RGB=&MakeVectorFromLinearColorRGB(Color=&MergeHelper_ColorSwitch(
      0=&LerpLinearColors(
       StartColor=<rgba 1 1 1 1>
       EndColor=&RandomRangeLinearColor(Minimum=<rgba 1 0.4 0.4 1>, Maximum=<rgba 0.4 1 0.4 1>)
       LerpFactor=&Waveform("Waveform [1]"=::NewEnumerator2, "[1] Frequency"=1, "[1] Amplitude Min/Max"=<0 1.2>, "[1] Scale"=$System.AudioLow)
      )
      1=&LerpLinearColors(
       StartColor=<rgba 1 1 1 1>
       EndColor=&RandomRangeLinearColor(Minimum=<rgba 0.4 1 0.4 1>, Maximum=<rgba 0.4 0.4 1 1>)
       LerpFactor=&Waveform("Waveform [1]"=::NewEnumerator2, "[1] Frequency"=1.5, "[1] Amplitude Min/Max"=<0 1.2>, "[1] Scale"=$System.AudioMedium)
      )
      2=&LerpLinearColors(
       StartColor=<rgba 1 1 1 1>
       EndColor=&RandomRangeLinearColor(Minimum=<rgba 0.4 0.4 1 1>, Maximum=<rgba 1 0.4 0.4 1>)
       LerpFactor=&Waveform("Waveform [1]"=::NewEnumerator2, "[1] Frequency"=2, "[1] Amplitude Min/Max"=<0 1.2>, "[1] Scale"=$System.AudioHigh)
      )
      3=&LerpLinearColors(
       StartColor=<rgba 1 1 1 1>
       EndColor=&RandomRangeLinearColor(Minimum=<rgba 1 0.4 0.4 1>, Maximum=<rgba 0.4 1 0.4 1>)
       LerpFactor=&Waveform("Waveform [1]"=::NewEnumerator2, "[1] Frequency"=1, "[1] Amplitude Min/Max"=<0 1.2>, "[1] Scale"=$System.AudioLow)
      )
      Index=$Particles.PartitionParticles.Partition
      Num=::NewEnumerator2
     ))
    ScaleA=true
    ScaleRGB=true
   ) {module=ScaleColor001}
   &ScaleMeshSize(Scale Factor=&MakeVector(
     X=&Waveform("Waveform [1]"=::NewEnumerator2, "[1] Frequency"=1, "[1] Amplitude Min/Max"=<-0.25 0.25>, "[1] Scale"=$System.AudioLow, "[1] Bias"=1)
     Y=&Waveform("Waveform [1]"=::NewEnumerator2, "[1] Frequency"=2, "[1] Amplitude Min/Max"=<-0.25 0.25>, "[1] Scale"=$System.AudioMedium, "[1] Bias"=1)
     Z=&Waveform("Waveform [1]"=::NewEnumerator2, "[1] Frequency"=3, "[1] Amplitude Min/Max"=<-0.25 0.25>, "[1] Scale"=$System.AudioHigh, "[1] Bias"=1)
    ))
   &ScaleMeshSize(Scale Factor=&VectorFromFloat(Value=&Lerp_Float(
      A=&Lerp_Float(A=$System.FrontEndReactiveScale, B=1, Alpha=$System.Timeline.NormalizedTime)
      B=&FN_FrontEndFloatModifier(Gameplay=1, Front End=1.1)
      Alpha=$System.reactiveFloat
     ))) {module=ScaleMeshSize002}
   &DynamicMaterialParameters(
    Index 0 Param 1=$System.EmissiveIntensity
    Index 0 Param 4=&MergeHelper_FloatSwitch(
     0=&Waveform("Waveform [1]"=::NewEnumerator2, "[1] Frequency"=1, "[1] Amplitude Min/Max"=<0 1>, "[1] Scale"=$System.AudioLow)
     1=&Waveform("Waveform [1]"=::NewEnumerator2, "[1] Frequency"=1.5, "[1] Amplitude Min/Max"=<0 1>, "[1] Scale"=$System.AudioMedium)
     2=&Waveform("Waveform [1]"=::NewEnumerator2, "[1] Frequency"=2, "[1] Amplitude Min/Max"=<0 1>, "[1] Scale"=$System.AudioHigh)
     3=&Waveform("Waveform [1]"=::NewEnumerator2, "[1] Frequency"=1, "[1] Amplitude Min/Max"=<0 1>, "[1] Scale"=$System.AudioLow)
     Index=$Particles.PartitionParticles.Partition
     Num=::NewEnumerator2
    )
    Module.Index0Float0Write=true
    Module.Index0Float1Write=true
    Module.Index0Float2Write=true
    Module.Index0Float3Write=true
   ) {module=DynamicMaterialParameters001}
  ]
  rendererClasses [&NiagaraMeshRendererProperties]
  renderers [{rendererClass=&NiagaraMeshRendererProperties, rendererIndex=0}]
 }
 Spiral_Particles_TravelingLoop003 {
  emitterUpdateScript [
   &EmitterState()
   &SpawnBurst_Instantaneous(Spawn Count=25, Use Spawn Probability=false, UseLoopCountLimit=true)
   /SetVariables(Emitter.Height=60, Emitter.SpiralFrequency=1.5, Emitter.TravelSpeed=0.12) {isSetParameters=true, module=SetVariables_#9}
  ]
  particleSpawnScript [
   &PartitionParticles(Number Of Partitions=4, Partition Mode:ENiagara_PartitionMode_v2=::NewEnumerator5)
   &InitializeParticle(
    Lifetime=5
    Mesh Scale Mode=::NewEnumerator1
    Mesh Uniform Scale Max=0.14
    Mesh Uniform Scale Min=0.09
    UsePositionOffset=false
    Write Material Random=true
   )
   &InitialMeshOrientation(
    Mesh Orientation Mode=::NewEnumerator2
    Rotation=&RandomRangeVector(Minimum=<0 -0.05 -0.05>, Maximum=<1 0.05 0.05>)
    Use Rotation Vector=true
   )
   /SetVariables(Particles.TravelSpeedParticle:NiagaraFloat=&FloatFromCurve(
     FloatCurve=DataInterface() {propertyValues {Curve {keys [<at 0 1>,<at 0.7931 1>,<at 1 0.7>]}}}
     CurveIndex=&RandomRangeFloat(Minimum=0, Maximum=1)
     Scale Curve=&MergeHelper_FloatSwitch(0=0.15, 1=0.1, 2=0.1, 3=0.1, Index=$Particles.PartitionParticles.Partition, Num=::NewEnumerator2)
    )) {isSetParameters=true, module=SetVariables_#12}
   &CameraOffset(Camera Offset Amount=3)
   &SkeletalMeshLocation(
    Bone / Socket Index=1
    MaskSpawnGroup=false
    Mesh Sampling Type=::NewEnumerator5
    Orientation Sampling=::NewEnumerator1
    Skeletal Mesh=$System.Skeletal Mesh
    Socket Sampling Mode=::NewEnumerator5
   )
  ]
  particleUpdateScript [
   &ParticleState(Kill Particles When Lifetime Has Elapsed=false, Let Infinitely Lived Particles Die When Emitter Deactivates=true)
   /SetVariables(
    Particles.CustomIndex:NiagaraFloat=&Frac_Float(Float=&Add_Float(
      A=&ReturnNormalizedExecIndex()
      B=&Multiply_Float(A=$Engine.Time, B=$Particles.TravelSpeedParticle)
     ))
    Particles.SpiralStrengthWaveform:NiagaraFloat=&Waveform(
     Global Input Position=&ReturnNormalizedExecIndex()
     "Waveform [1]"=::NewEnumerator0
     "[1] Phase Scale"=0.25
     "[1] Frequency"=1
     "[1] Amplitude Min/Max"=<1 1.1>
     "[1] Scale"=19.5
    )
   ) {isSetParameters=true, module=SetVariables_#10}
   /SetVariables(Particles.SpiralStrength:NiagaraFloat=&FloatFromCurve(
     FloatCurve=DataInterface() {propertyValues {Curve {
        defaultValue=1
        keys [{arriveTangent=-0.4468, interpMode=RCIM_Cubic, leaveTangent=-0.4468, tangentMode=RCTM_User, time=0, value=0.4487}, <at 1 0.35 RCIM_Cubic>]
       }}}
     CurveIndex=$Particles.CustomIndex
     Scale Curve=$Particles.SpiralStrengthWaveform
    )) {isSetParameters=true, module=SetVariables_#13}
   /SetVariables(Particles.SpiralFrequency:NiagaraFloat=&FloatFromCurve(
     FloatCurve=DataInterface() {propertyValues {Curve {keys [<at 0 1>,<at 1 1.2707>]}}}
     CurveIndex=$Particles.CustomIndex
     Scale Curve=$Emitter.SpiralFrequency
    )) {isSetParameters=true, module=SetVariables_#14}
   /SetVariables(Particles.PositionOffset:NiagaraPosition=&AddVectorToPosition(
     Position=&AddVectorToPosition(
      Position=$Particles.Position
      Vector=&Multiply_VectorByFloat(
       Vector=&RandomRangeVector(Minimum=<-0.75 -0.75 -1>, Maximum=<0.75 0.75 1>)
       Float=&Add_Float(
        A=&FloatFromCurve(FloatCurve=DataInterface() {propertyValues {Curve {keys [<at 0 1>,<at 1 0.1>]}}}, CurveIndex=$Particles.CustomIndex, Scale Curve=4)
        B=&FloatFromCurve(
         FloatCurve=DataInterface() {propertyValues {Curve {
            defaultValue=1
            keys [
             {arriveTangent=5.6707, interpMode=RCIM_Cubic, leaveTangent=5.6707, tangentMode=RCTM_User, time=0, value=0.538}
             {arriveTangentWeight=0.3222, interpMode=RCIM_Cubic, leaveTangentWeight=0.0504, tangentMode=RCTM_User, time=0.1717, value=1.0736}
             {
              arriveTangent=-2.0781
              arriveTangentWeight=0.0226
              interpMode=RCIM_Cubic
              leaveTangent=-1.0421
              leaveTangentWeight=0.0234
              tangentMode=RCTM_Break
              time=0.59
              value=0.5528
             }
             {
              arriveTangent=-0.5117
              arriveTangentWeight=0.0523
              interpMode=RCIM_Cubic
              leaveTangent=-0.5117
              leaveTangentWeight=0.3204
              tangentMode=RCTM_User
              time=0.8408
              value=0.7639
             }
             {arriveTangent=-2.4831, interpMode=RCIM_Cubic, leaveTangent=-2.4831, tangentMode=RCTM_User, time=1, value=0.4}
            ]
           }}}
         CurveIndex=$Particles.CustomIndex
         Scale Curve=&Waveform(
          Global Input Position=&ReturnNormalizedExecIndex()
          "Waveform [1]"=::NewEnumerator2
          "[1] Phase Scale"=0.2
          "[1] Frequency"=1.1
          "[1] Amplitude Min/Max"=<-1 1>
          "[1] Scale"=&Lerp_Float(A=4, B=6, Alpha=$System.reactiveFloat)
         )
        )
       )
      )
     )
     Vector=&MakeVector(
      X=&Waveform(
       Global Input Position=$Particles.CustomIndex
       Global Phase=&Add_Float(A=&Multiply_Float(A=$Engine.Time, B=$Emitter.SpiralSpeed), B=$Emitter.SpiralOffset)
       "Waveform [1]"=::NewEnumerator0
       "[1] Frequency"=$Particles.SpiralFrequency
       "[1] Amplitude Min/Max"=<-1 1>
       "[1] Scale"=$Particles.SpiralStrength
      )
      Y=&Waveform(
       Global Input Position=$Particles.CustomIndex
       Global Phase=&Add_Float(A=&Multiply_Float(A=$Engine.Time, B=$Emitter.SpiralSpeed), B=$Emitter.SpiralOffset)
       "Waveform [1]"=::NewEnumerator1
       "[1] Frequency"=$Particles.SpiralFrequency
       "[1] Amplitude Min/Max"=<-1 1>
       "[1] Scale"=$Particles.SpiralStrength
      )
      Z=&FloatFromCurve(
       FloatCurve=DataInterface() {propertyValues {Curve {defaultValue=0, keys [{tangentMode=RCTM_User, time=0, value=0}, {tangentMode=RCTM_User, time=1, value=1}]}}}
       CurveIndex=$Particles.CustomIndex
       Scale Curve=&Multiply_Float(A=$Emitter.Height, B=1)
      )
     )
    )) {isSetParameters=true, module=SetVariables_#11}
   &ScaleMeshSize(Scale Factor=&Multiply_VectorByFloat(
     Vector=<1 1 1>
     Float=&Multiply_Float(
      A=&Multiply_Float(
       A=&MergeHelper_FloatSwitch(
        0=&Waveform(
         "Waveform [1]"=::NewEnumerator2
         "[1] Frequency"=1
         "[1] Amplitude Min/Max"=<-0.25 0.75>
         "[1] Scale"=$System.AudioLow
         "[1] Bias"=&RandomRangeFloat(Minimum=0.8, Maximum=1.1)
        )
        1=&Waveform(
         "Waveform [1]"=::NewEnumerator2
         "[1] Frequency"=1.5
         "[1] Amplitude Min/Max"=<-0.25 0.75>
         "[1] Scale"=$System.AudioMedium
         "[1] Bias"=&RandomRangeFloat(Minimum=0.8, Maximum=1.1)
        )
        2=&Waveform(
         "Waveform [1]"=::NewEnumerator2
         "[1] Frequency"=2
         "[1] Amplitude Min/Max"=<-0.25 0.75>
         "[1] Scale"=$System.AudioHigh
         "[1] Bias"=&RandomRangeFloat(Minimum=0.8, Maximum=1.1)
        )
        3=&Waveform(
         "Waveform [1]"=::NewEnumerator2
         "[1] Frequency"=1
         "[1] Amplitude Min/Max"=<-0.25 0.75>
         "[1] Scale"=$System.AudioLow
         "[1] Bias"=&RandomRangeFloat(Minimum=0.8, Maximum=1.1)
        )
        Index=$Particles.PartitionParticles.Partition
        Num=::NewEnumerator2
       )
       B=1
      )
      B=&FloatFromCurve(
       FloatCurve=DataInterface() {propertyValues {Curve {
          defaultValue=0
          keys [
           {arriveTangent=14.706, interpMode=RCIM_Cubic, leaveTangent=14.706, leaveTangentWeight=0.0825, tangentMode=RCTM_User, time=0, value=0}
           {
            arriveTangent=1.3013
            arriveTangentWeight=0.1782
            interpMode=RCIM_Cubic
            leaveTangent=1.3013
            leaveTangentWeight=0.1222
            tangentMode=RCTM_User
            time=0.1323
            value=0.8
           }
           {
            arriveTangent=-2.2321
            arriveTangentWeight=0.0723
            interpMode=RCIM_Cubic
            leaveTangent=-2.2321
            leaveTangentWeight=0.0438
            tangentMode=RCTM_User
            time=0.2752
            value=0.6597
           }
           {
            arriveTangent=-0.6475
            arriveTangentWeight=0.1638
            interpMode=RCIM_Cubic
            leaveTangent=-0.6475
            leaveTangentWeight=0.2008
            tangentMode=RCTM_User
            time=0.401
            value=0.7
           }
           {
            arriveTangent=-9.6816
            arriveTangentWeight=0.1494
            interpMode=RCIM_Cubic
            leaveTangentWeight=0.0424
            tangentMode=RCTM_Break
            time=0.65
            value=0.1
           }
           {
            arriveTangentWeight=0.0424
            interpMode=RCIM_Cubic
            leaveTangent=12.3539
            leaveTangentWeight=0.0531
            tangentMode=RCTM_Break
            time=0.7074
            value=0.1
           }
           {
            arriveTangent=0.1355
            arriveTangentWeight=0.0769
            interpMode=RCIM_Cubic
            leaveTangent=0.1355
            leaveTangentWeight=0.054
            tangentMode=RCTM_User
            time=0.8622
            value=0.6
           }
           {arriveTangent=-12.6776, interpMode=RCIM_Cubic, leaveTangent=-12.6776, tangentMode=RCTM_User, time=1, value=0}
          ]
         }}}
       CurveIndex=$Particles.CustomIndex
      )
     )
    )) {module=ScaleMeshSize001}
   &ScaleColor(
    Scale RGB=&MakeVectorFromLinearColorRGB(Color=&MergeHelper_ColorSwitch(
      0=&LerpLinearColors(
       StartColor=<rgba 1 1 1 1>
       EndColor=&RandomRangeLinearColor(Minimum=<rgba 1 0.4 0.4 1>, Maximum=<rgba 0.4 1 0.4 1>)
       LerpFactor=&Waveform("Waveform [1]"=::NewEnumerator2, "[1] Frequency"=1, "[1] Amplitude Min/Max"=<0 1.2>, "[1] Scale"=$System.AudioLow)
      )
      1=&LerpLinearColors(
       StartColor=<rgba 1 1 1 1>
       EndColor=&RandomRangeLinearColor(Minimum=<rgba 0.4 1 0.4 1>, Maximum=<rgba 0.4 0.4 1 1>)
       LerpFactor=&Waveform("Waveform [1]"=::NewEnumerator2, "[1] Frequency"=1.5, "[1] Amplitude Min/Max"=<0 1.2>, "[1] Scale"=$System.AudioMedium)
      )
      2=&LerpLinearColors(
       StartColor=<rgba 1 1 1 1>
       EndColor=&RandomRangeLinearColor(Minimum=<rgba 0.4 0.4 1 1>, Maximum=<rgba 1 0.4 0.4 1>)
       LerpFactor=&Waveform("Waveform [1]"=::NewEnumerator2, "[1] Frequency"=2, "[1] Amplitude Min/Max"=<0 1.2>, "[1] Scale"=$System.AudioHigh)
      )
      3=&LerpLinearColors(
       StartColor=<rgba 1 1 1 1>
       EndColor=&RandomRangeLinearColor(Minimum=<rgba 1 0.4 0.4 1>, Maximum=<rgba 0.4 1 0.4 1>)
       LerpFactor=&Waveform("Waveform [1]"=::NewEnumerator2, "[1] Frequency"=1, "[1] Amplitude Min/Max"=<0 1.2>, "[1] Scale"=$System.AudioLow)
      )
      Index=$Particles.PartitionParticles.Partition
      Num=::NewEnumerator2
     ))
    ScaleA=true
    ScaleRGB=true
   ) {module=ScaleColor001}
   &ScaleMeshSize(Scale Factor=&MakeVector(
     X=&Waveform("Waveform [1]"=::NewEnumerator2, "[1] Frequency"=1, "[1] Amplitude Min/Max"=<-0.25 0.25>, "[1] Scale"=$System.AudioLow, "[1] Bias"=1)
     Y=&Waveform("Waveform [1]"=::NewEnumerator2, "[1] Frequency"=2, "[1] Amplitude Min/Max"=<-0.25 0.25>, "[1] Scale"=$System.AudioMedium, "[1] Bias"=1)
     Z=&Waveform("Waveform [1]"=::NewEnumerator2, "[1] Frequency"=3, "[1] Amplitude Min/Max"=<-0.25 0.25>, "[1] Scale"=$System.AudioHigh, "[1] Bias"=1)
    ))
   &ScaleMeshSize(Scale Factor=&VectorFromFloat(Value=&Lerp_Float(
      A=&Lerp_Float(A=$System.FrontEndReactiveScale, B=1, Alpha=$System.Timeline.NormalizedTime)
      B=&FN_FrontEndFloatModifier(Gameplay=1, Front End=1.1)
      Alpha=$System.reactiveFloat
     ))) {module=ScaleMeshSize002}
   &DynamicMaterialParameters(
    Index 0 Param 1=$System.EmissiveIntensity
    Index 0 Param 4=&MergeHelper_FloatSwitch(
     0=&Waveform("Waveform [1]"=::NewEnumerator2, "[1] Frequency"=1, "[1] Amplitude Min/Max"=<0 1>, "[1] Scale"=$System.AudioLow)
     1=&Waveform("Waveform [1]"=::NewEnumerator2, "[1] Frequency"=1.5, "[1] Amplitude Min/Max"=<0 1>, "[1] Scale"=$System.AudioMedium)
     2=&Waveform("Waveform [1]"=::NewEnumerator2, "[1] Frequency"=2, "[1] Amplitude Min/Max"=<0 1>, "[1] Scale"=$System.AudioHigh)
     3=&Waveform("Waveform [1]"=::NewEnumerator2, "[1] Frequency"=1, "[1] Amplitude Min/Max"=<0 1>, "[1] Scale"=$System.AudioLow)
     Index=$Particles.PartitionParticles.Partition
     Num=::NewEnumerator2
    )
    Module.Index0Float0Write=true
    Module.Index0Float1Write=true
    Module.Index0Float2Write=true
    Module.Index0Float3Write=true
   ) {module=DynamicMaterialParameters001}
  ]
  rendererClasses [&NiagaraMeshRendererProperties]
  renderers [{rendererClass=&NiagaraMeshRendererProperties, rendererIndex=0}]
 }
}
summary {
 systemName=NS_DrumKit_AnvilAlarmMoth_IdleFX
 emitters [
  {
   bEnabled=false
   emitterName=Spiral_Particles_Constant
   rendererClasses [&NiagaraSpriteRendererProperties, &NiagaraMeshRendererProperties]
  }
  {
   bEnabled=false
   emitterName=Spiral_Particles_SpawnRate
   rendererClasses [&NiagaraSpriteRendererProperties, &NiagaraMeshRendererProperties]
  }
  {emitterName=Spiral_Particles_TravelingLoop, rendererClasses [&NiagaraMeshRendererProperties]}
  {emitterName=Spiral_Particles_TravelingLoop001, rendererClasses [&NiagaraMeshRendererProperties]}
  {emitterName=Spiral_Particles_TravelingLoop002, rendererClasses [&NiagaraMeshRendererProperties]}
  {emitterName=Spiral_Particles_TravelingLoop003, rendererClasses [&NiagaraMeshRendererProperties]}
 ]
 userVariables [
  User.bIsFrontEnd {type=NiagaraBool}
  User.bIsFrontEndPreview {type=NiagaraBool}
 ]
}
